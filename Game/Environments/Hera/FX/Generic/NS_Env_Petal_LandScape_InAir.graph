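# Niagara environment effect that fills the air above landscape terrain with drifting, hue-varied petals, using one GPU emitter with sprite and mesh renderers.
# Spawn rate and a sweeping spawn-sphere offset follow the system loop, and landscape sampling places petals at random heights above the ground.
# Curl noise, wind and aerodynamic drag move them, with noise gated by a latched terrain-height check.
# Owner LOD distance scales size to zero at the cull distance.
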NiagaraSystem /Game/Environments/Hera/FX/Generic/NS_Env_Petal_LandScape_InAir
dependencies {
 usedDataInterfaces [&NiagaraDataInterfaceCurve, &NiagaraDataInterfaceCollisionQuery]
 usedDynamicInputs []
 usedModules [&SystemState]
}
emitters {Petals {
  simTarget=GPUComputeSim
  emitterUpdateScript [
   &EmitterState(
    Enable Distance Culling=true
    Max Distance Response=::SleepAndClearParticles
    MaxDistance=8000
    Module.bUseMaxDistance=true
    Module.bUseMinDistance=false
    Reset Age On Awaken=true
    Scalability Mode=::NewEnumerator1
   )
   &SpawnRate(
    SpawnRate=&FloatFromCurve(
     FloatCurve=DataInterface() {propertyValues {Curve {keys [<at 0 1>,<at 0.1 2>,<at 0.25 1>,<at 0.4976 0.25>,<at 0.8283 0.25>,<at 1 1>]}}}
     CurveIndex=$System.NormalizedLoopAge
     Scale Curve=$User.SpawnRate
    )
    Use Spawn Probability=false
   )
  ]
  particleSpawnScript [
   &InitializeParticle(
    Color=&HueShiftLinearColor(
     Hue=&RandomRangeFloat(Minimum=-0.05, Maximum=0.1)
     Saturation=&RandomRangeFloat(Minimum=0.5, Maximum=1)
     Value=2
     Linear Color=$User.UserColor
    )
    Lifetime=&RandomRangeFloat(Minimum=2, Maximum=5)
    Mass=3
    Mass Mode=::NewEnumerator1
    Mesh Scale Mode=::NewEnumerator1
    Mesh Uniform Scale Max=0.05
    Mesh Uniform Scale Min=0.04
    Sprite Rotation Mode=::NewEnumerator2
    Sprite Size Mode=::NewEnumerator1
    UsePositionOffset=false
    Write Material Random=true
   ) {module=InitializeParticle001}
   /SetVariables(Particles.Terrain Offset:NiagaraFloat=&RandomRangeFloat(Minimum=25, Maximum=200)) {isSetParameters=true}
   &ShapeLocation(
    MaskSpawnGroup=false
    Non Uniform Scale=<0.5 1 1>
    Offset=&MakeVector(
     X=&FloatFromCurve(FloatCurve=DataInterface() {propertyValues {Curve {keys [<at 0 -1>,<at 1 2>]}}}, CurveIndex=$System.NormalizedLoopAge, Scale Curve=250)
     Y=0
     Z=0
    )
    Offset Mode=::NewEnumerator0
    Override Seed=false
    Sphere Radius=$User.SphereLocation
    UseOwnerScale=true
   )
   &self:Sample_Landscape_01() {module=Sample_Landscape_01}
   /SetVariables(Particles.Position=&Add_Vector(
     A=Linked() {linkedVariable {name=Particles.Position, type=NiagaraPosition}}
     B=&MakeVector(X=0, Y=0, Z=$Particles.Terrain Offset)
    )) {isSetParameters=true}
   /SetVariables(Particles.New Vector 3f:Vector3f=<0 0 1>) {isSetParameters=true}
   &CameraOffset(Camera Offset Amount=15)
   &SubUVAnimation(SubUV Animation Mode:ENiagara_SubUVLookupMode=::NewEnumerator2) {module=SubUVAnimation001}
   /SetVariables(Particles.SubImageIndex=&RandomRangeFloat(Minimum=0, Maximum=63)) {isSetParameters=true}
   /SetVariables() {isSetParameters=true}
   &InitialMeshOrientation(Rotation Coordinate Space=::NewEnumerator2, Use Rotation Vector=true)
  ]
  particleUpdateScript [
   &self:Sample_Landscape_UPdate_01(Bias:NiagaraFloat=$Particles.Terrain Offset, New Landscape Sample:NiagaraDataInterfaceLandscape=$User.Landscape) {module=Sample_Landscape_UPdate_01}
   &ScaleColor(
    Scale Alpha=&FloatFromCurve(FloatCurve=DataInterface() {propertyValues {Curve {keys [<at 0.8 1>,<at 1 0>]}}}, Scale Curve=5)
    Scale RGB=&MakeVectorFromLinearColorRGB(Color=&MultiplyLinearColor(Linear Color=$Particles.Initial.Color, Scale Factor=&MakeVector4_FromFloat(Float=1)))
    ScaleA=true
    ScaleRGB=true
   )
   /SetVariables(Particles.Normalized Velocity:Vector3f=&NormalizeVector(Vector To Normalize=$Particles.Velocity)) {isSetParameters=true}
   /SetVariables(Particles.Is Active:NiagaraBool=&Boolean_OrOperation(
     Bool1=$Particles.Is Active
     Bool2=&SetBoolByFloatComparison(
      A=&MakeFloatFromVector(Vector=Linked() {linkedVariable {name=Particles.Position, type=NiagaraPosition}}, Channel=::NewEnumerator2)
      B=$Output.Sample_Landscape_UPdate_01.Terrain Height
      Comparison Type=::NewEnumerator6
     )
    )) {isSetParameters=true}
   &CurlNoiseForce(
    Noise Frequency=25
    Noise Strength=&MakeCustomFloatFromBool(
     Bool=$Particles.Is Active
     True Float=&FloatFromCurve(FloatCurve=DataInterface() {propertyValues {Curve {keys [<at 0 1>,<at 1 0.1>]}}}, Scale Curve=500)
     False Float=0
    )
    Pan Noise=false
   ) {module=CurlNoiseForce001}
   &SolveForcesAndVelocity(Clamp Velocity=true, Limit Acceleration=false, Speed Limit=250)
   &self:ScratchModule_05(Terrain Height:NiagaraFloat=$Output.Sample_Landscape_UPdate_01.Terrain Height, Threshold=50) {module=ScratchModule_05}
   &SubUVAnimation(End Frame=64, SubUV Animation Mode:ENiagara_SubUVLookupMode=::NewEnumerator1, SubUV Loop Count=4)
   &ParticleState(DeltaTime:NiagaraFloat=&MakeCustomFloatFromBool(
     Bool=&SetBoolByFloatComparison(A=&VectorLength(Vector=$Particles.Velocity), B=1)
     True Float=$Engine.DeltaTime
     False Float=0
    ))
   &ScaleMeshSize(Scale Factor=&VectorFromFloat(Value=&Multiply_Float(
      A=&FloatFromCurve(
       FloatCurve=DataInterface() {propertyValues {Curve {keys [<at 0.25 0.25>,<at 500 1>,<at 6000 1>,<at 8000 0>]}}}
       CurveIndex=$Engine.Owner.LODDistance
       Scale Curve=0.5
      )
      B=&FloatFromCurve(FloatCurve=DataInterface() {propertyValues {Curve {keys [<at 0.8 1>,<at 1 0>]}}}, Scale Curve=0.25)
     )))
   &ScaleSpriteSize(
    Scale Sprite Size Mode=::NewEnumerator0
    Uniform Scale Factor=&Multiply_Float(
     A=&FloatFromCurve(
      FloatCurve=DataInterface() {propertyValues {Curve {keys [<at 0 1>,<at 500 1>,<at 6000 1>,<at 8000 0>]}}}
      CurveIndex=$Engine.Owner.LODDistance
      Scale Curve=2
     )
     B=&FloatFromCurve(FloatCurve=DataInterface() {propertyValues {Curve {keys [<at 0.8 1>,<at 1 0>]}}}, Scale Curve=0.25)
    )
   )
   &WindForce(
    Frequency=0.25
    Friction Mode:ENiagaraWindFrictionMode=::NewEnumerator0
    Offset By Time:Vector3f=<0 0 -10>
    Scale=0.1
    Use Ground Mask=true
    Wind Speed=<250 0 250>
    Wind Speed Scale=&RandomRangeFloat(Minimum=0.5, Maximum=1)
   ) {module=WindForce001}
   &AerodynamicDrag(
    Aerodynamic Drag=0.1
    Aerodynamic Rotational Drag=0.1
    Drag Shape Curve=DataInterface() {propertyValues {Curve {keys [
        {interpMode=RCIM_Cubic, tangentMode=RCTM_User, time=0, value=0.1}
        {arriveTangent=2.0799, interpMode=RCIM_Cubic, leaveTangent=2.0799, tangentMode=RCTM_User, time=1, value=1}
       ]}}}
    Lift Shape Curve=DataInterface() {propertyValues {Curve {keys [
        {interpMode=RCIM_Cubic, tangentMode=RCTM_User, time=0, value=0}
        {interpMode=RCIM_Cubic, tangentMode=RCTM_User, time=0.5, value=1}
        {tangentMode=RCTM_User, time=1, value=0}
       ]}}}
    Pivot Offset Maximum:Vector3f=<1 1 1>
    Pivot Offset Minimum Length:Vector3f=<0 0 0>
    Rotation Strength:NiagaraFloat=&RandomRangeFloat(Minimum=15, Maximum=25)
    Use Linear Drag=true
    Use Rotational Drag=true
   )
   &SolveForcesAndVelocity(
    Clamp Velocity=true
    Limit Acceleration=false
    Speed Limit:NiagaraFloat=&RandomRangeFloat(Minimum=35, Maximum=50)
   ) {module=SolveForcesAndVelocity001}
   &AlignSpriteToMeshOrientation(
    Mesh Orientation Relative Sprite Alignment Vector=<1 0 0>
    Mesh Orientation Relative Sprite Facing Vector:Vector3f=<1 1 1>
    SetAlignment=true
    SetFacing=true
   )
   &DynamicMaterialParameters(
    Index 0 Param 1=&RandomRangeFloat(Minimum=0.02, Maximum=0.1)
    Index 0 Param 2=&RandomRangeFloat(Minimum=-0.05, Maximum=0.05)
    Index 0 Param 3=1
    Module.Param0WriteEnabled=true
    Module.Param1WriteEnabled=true
    Module.Param2WriteEnabled=true
    Module.Param3WriteEnabled=true
   )
   /SetVariables(Particles.SpriteColor=&MultiplyLinearColor(Linear Color=$Particles.Color, Scale Factor=&MakeVector4_FromFloat(Float=2))) {isSetParameters=true}
  ]
  rendererClasses [&NiagaraSpriteRendererProperties, &NiagaraMeshRendererProperties]
  renderers [
   {rendererClass=&NiagaraSpriteRendererProperties}
   {rendererClass=&NiagaraMeshRendererProperties, rendererIndex=1}
  ]
 }}
summary {
 systemName=NS_Env_Petal_LandScape_InAir
 emitters [{
   emitterName=Petals
   simTarget=GPUComputeSim
   rendererClasses [&NiagaraSpriteRendererProperties, &NiagaraMeshRendererProperties]
  }]
 userVariables [
  User.Alpha_AttractionStrength {type=NiagaraFloat, defaultValue=0, description=""}
  User.Landscape {
   type=NiagaraDataInterfaceLandscape
   description=""
   defaultValue {
    struct=NiagaraExt_VariableValue_DataInterface
    value {
     dataInterface=&self:NiagaraDataInterfaceLandscape_0
     dataInterfaceClass=&NiagaraDataInterfaceLandscape
    }
   }
  }
  User.Player Position {type=Vector3f, description="", defaultValue {struct=Vector3f, value=<0 0 0>}}
  User.SpawnRate {type=NiagaraFloat, defaultValue=750, description=""}
  User.SphereLocation {type=NiagaraFloat, defaultValue=2500, description=""}
  User.UserColor {type=LinearColor, description="", defaultValue {struct=LinearColor, value=<rgba 0.5341 0.4175 0.8714 1>}}
 ]
}
